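# Material function that defines a basic rubber surface layer, filed with the Superlevel DigitalCamo common functions for player character cosmetics.
# A Set Material Attributes node builds the surface from a tint parameter, dark by default, and a mid-range roughness parameter.
# Constants supply default specular, a flat Z-only normal and zero subsurface color.
# It samples no textures and outputs the attributes directly.
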MaterialFunction /BRCosmetics/Characters/Player/Common/MaterialFunctions/Superlevel_DigitalCamo/MF_SL_BasicRubber
graph {nodes [
  n1 = SetMaterialAttributes!(
   MaterialAttributes
   BaseColor=VectorParameter() {parameterName=Rubber_Tint, defaultValue=<rgba 0.02 0.02 0.02 1>}.RGB
   Roughness=ScalarParameter() {defaultValue=0.55, parameterName=RubberRoughness}
   Specular=Constant() {r=0.5}
   Normal=Constant3Vector() {constant=<rgba 0 0 0.5 0>}
   SubsurfaceColor=Constant()
  ) {attributeSetTypes [
    #1
    #2
    #3
    #4
    #5
   ]}
  FunctionOutput(None=@n1)
 ]}
