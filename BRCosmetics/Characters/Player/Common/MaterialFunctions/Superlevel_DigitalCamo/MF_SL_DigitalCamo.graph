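# Material function that applies an animated three-tone "Superlevel" digital camo to player cosmetic materials.
# Three panning projections of one mask texture in pre-skinned space are thresholded against a time-driven sine to blend colors and roughness, with sine stand-ins on low/mobile.
# Fresnel and the GMap input drive color-shifting emissive, while Camo_Pattern_Fade flattens the pattern to white and removes the glow.
# ThinFilmFunction finishes the attributes, skipped on mobile when isBinaryHelm is set.
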MaterialFunction /BRCosmetics/Characters/Player/Common/MaterialFunctions/Superlevel_DigitalCamo/MF_SL_DigitalCamo
graph {nodes [
  Abs()
  n1 = TextureObject() {samplerType=SAMPLERTYPE_LinearColor, texture=/Game/Characters/Player/Common/MaterialFunctions/Superlevel_DigitalCamo/SuperCamo_M}
  n2 = FunctionInput(Preview=VertexInterpolator(VS=PreSkinnedPosition()).PS) {bUsePreviewValueAsDefault=true, inputName=PreSkinnedLocal}
  n3 = ScalarParameter() {parameterName=Pan1_X}
  n4 = ScalarParameter() {parameterName=Pan1_Y}
  n5 = VertexInterpolator(VS=PreSkinnedNormal())
  n6 = Constant() {r=5}
  n7 = MaterialFunctionCall!(
   Default=MaterialFunctionCall!(
    TextureObject=@n1
    TextureSize=ScalarParameter() {defaultValue=18, parameterName=CamoScale}
    WorldPosition=@n2
    Export Float 4
    In=AppendVector(A=@n3, B=@n4)
    World Space Normal=@n5.PS
    ProjectionTransitionContrast=@n6
   ) {materialFunction=/Game/Athena/SafeZone/WorldAlignedTexture_ScalePan}.XYZ Texture
   Low/Mobile=Sine(None=Time()) {period=350}
  ) {materialFunction=/Game/Effects/Fort_Effects/Materials/_Functions/MF_LowMobile_PathingSwitch}
  n8 = ScalarParameter() {parameterName=Pan3_X}
  n9 = ScalarParameter() {defaultValue=0.05, parameterName=Pan3_Y}
  n10 = MaterialFunctionCall(
   Default=MaterialFunctionCall(
    TextureObject=@n1
    TextureSize=ScalarParameter() {defaultValue=30, parameterName=CamoScale2}
    WorldPosition=@n2
    In=AppendVector(A=@n8, B=@n9)
    World Space Normal=@n5.PS
    ProjectionTransitionContrast=@n6
   ) {materialFunction=&WorldAlignedTexture_ScalePan}.XYZ Texture
   Low/Mobile=Sine(None=Time()) {period=400}
  ) {materialFunction=&MF_LowMobile_PathingSwitch}
  n11 = Subtract(A=Subtract(A=ComponentMask(None=@n7.Result) {r=true}, B=ComponentMask(None=@n10.Result) {b=true}))
  n12 = ScalarParameter() {defaultValue=0.025, parameterName=Adjustment Speed}
  n13 = Time() {period=0.5}
  n14 = Abs(None=Sine(None=Multiply(A=@n12, B=@n13)))
  n15 = OneMinus(None=Clamp(None=@n14) {maxDefault=0.8, minDefault=0.18})
  n16 = Saturate(None=Ceil(None=Add(A=@n11, B=@n15)))
  Ceil(None=@n16)
  Constant() {r=1}
  Frac()
  n17 = GetMaterialAttributes(None=FunctionInput() {inputType=FunctionInput_MaterialAttributes}) {attributeGetTypes [#1]}
  n18 = VectorParameter() {parameterName=Camo_Base, defaultValue=<rgba 0.8438 0.8438 0.8438 1>}
  n19 = VectorParameter() {parameterName=Camo_Mid, defaultValue=<rgba 1 1 1 1>}
  n20 = ScalarParameter() {defaultValue=20, parameterName=CamoScale3}
  n21 = AppendVector(A=ScalarParameter() {parameterName=Pan2_X}, B=ScalarParameter() {defaultValue=-0.058, parameterName=Pan2_Y})
  n22 = Time()
  n23 = ComponentMask(None=MaterialFunctionCall(
    Default=MaterialFunctionCall(TextureObject=@n1, TextureSize=@n20, WorldPosition=@n2, In=@n21, World Space Normal=@n5.PS, ProjectionTransitionContrast=@n6) {materialFunction=&WorldAlignedTexture_ScalePan}.XYZ Texture
    Low/Mobile=Sine(None=@n22) {period=650}
   ) {materialFunction=&MF_LowMobile_PathingSwitch}.Result) {g=true}
  n24 = Add(A=Subtract(A=@n23), B=Clamp(None=@n14) {maxDefault=0.9, minDefault=0.1})
  n25 = Ceil(None=@n24)
  n26 = VectorParameter() {parameterName=Camo_White, defaultValue=<rgba 1 1 1 1>}
  n27 = Round(None=@n16)
  n28 = ScalarParameter() {parameterName=Camo_Pattern_Fade}
  n29 = LinearInterpolate(A=LinearInterpolate(A=LinearInterpolate(A=@n18.RGB, B=@n19.RGB, Alpha=@n25), B=@n26.RGB, Alpha=@n27), B=@n26.RGB, Alpha=@n28)
  n30 = FunctionInput(Preview=Fresnel() {baseReflectFraction=0, exponent=1}) {inputName=Fresnel, inputType=FunctionInput_Scalar}
  n31 = ConstantBiasScale(None=Sine(None=@n30)) {bias=0.5}
  n32 = LinearInterpolate(
   A=ScalarParameter() {defaultValue=0.55, parameterName=MinRoughness}
   B=ScalarParameter() {defaultValue=0.8, parameterName=MaxRoughness}
   Alpha=Max(A=@n16, B=Multiply(A=@n23) {constB=0.5})
  )
  n33 = Constant() {r=0.8}
  n34 = Power(Base=@n30, Exp=ScalarParameter() {defaultValue=3, parameterName=FresnelPower})
  n35 = Saturate(None=@n34)
  n36 = LinearInterpolate(A=Constant3Vector() {constant=<rgba 1 0.6498 0.4034 0>}, B=Constant3Vector() {constant=<rgba 0.3889 0.6983 1 0>}, Alpha=@n14)
  n37 = LinearInterpolate(A=@n35, B=Multiply(A=@n35, B=@n23), Alpha=@n14)
  n38 = Multiply(
   A=ScalarParameter() {defaultValue=3, parameterName=FresnelBrightness}
   B=Multiply(A=FunctionInput() {inputName=GMap, inputType=FunctionInput_Scalar}, B=LinearInterpolate(B=@n36, Alpha=@n37))
  )
  n39 = Power(Base=@n34) {constExponent=0.01}
  n40 = Multiply(A=Multiply(A=@n27, B=OneMinus(None=@n39)), B=ScalarParameter() {defaultValue=0.5, parameterName=White Layer Emissive})
  n41 = SetMaterialAttributes!(
   MaterialAttributes=@n17.MaterialAttributes
   BaseColor=Multiply(A=Saturate(None=@n29), B=LinearInterpolate(Alpha=@n31) {constA=1, constB=0.5})
   Metallic=ScalarParameter() {defaultValue=0.7, parameterName=CamoMetallic}
   Roughness=LinearInterpolate(A=LinearInterpolate(A=@n32, B=@n33, Alpha=@n35), B=ScalarParameter() {defaultValue=0.66, parameterName=Mid_Roughness}, Alpha=@n28)
   EmissiveColor=LinearInterpolate(A=Add(A=@n38, B=@n40), Alpha=@n28) {constB=0}
   Specular=ScalarParameter() {defaultValue=0.5, parameterName=CamoSpec}
   SubsurfaceColor=Constant()
  ) {attributeSetTypes [
    #2
    #3
    #4
    #5
    #6
    #7
   ]}
  n42 = MaterialFunctionCall!(Alpha=@n16, Material Attributes=@n41) {materialFunction=/Game/Weapons/WeaponSkins/MaterialFunction/ThinFilmFunction}
  FunctionOutput(None=StaticSwitchParameter(True=ShadingPathSwitch(Default=@n42.Result, Mobile=@n41) {desc=Migrated from FeatureLevelSwitch}, False=@n42.Result) {parameterName=isBinaryHelm})
  Round(None=@n24)
  Transform(None=@n2) {transformSourceType=TRANSFORMSOURCE_Local}
  Transform(None=@n17.Normal)
  VertexColor()
  VertexNormalWS()
 ]}
