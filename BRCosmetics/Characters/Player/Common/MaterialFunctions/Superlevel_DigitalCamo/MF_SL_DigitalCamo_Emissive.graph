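# Material function for the Superlevel Digital Camo player cosmetic that restyles incoming material attributes as a glossy, glowing emissive surface.
# It overrides base color to black and sets constant low roughness and high specular.
# Emissive is an HDR color parameter, cyan-blue by default, multiplied by one minus the Fresnel input raised to a tunable power.
# The modified attributes are the function's output.
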MaterialFunction /BRCosmetics/Characters/Player/Common/MaterialFunctions/Superlevel_DigitalCamo/MF_SL_DigitalCamo_Emissive
graph {nodes [
  n1 = Power(
   Base=FunctionInput() {inputName=Fresnel, inputType=FunctionInput_Scalar}
   Exp=ScalarParameter() {defaultValue=0.04, parameterName=EmissiveFresnelPower}
  )
  n2 = SetMaterialAttributes!(
   MaterialAttributes=FunctionInput() {inputType=FunctionInput_MaterialAttributes}
   BaseColor=Constant()
   Roughness=Constant() {r=0.15}
   Specular=Constant() {r=0.8}
   EmissiveColor=Multiply(A=OneMinus(None=@n1), B=VectorParameter() {parameterName=EmissiveColor, defaultValue=<rgba 12.5885 31.0745 40 1>}.RGB)
  ) {attributeSetTypes [
    #1
    #2
    #3
    #4
   ]}
  FunctionOutput(None=@n2)
 ]}
