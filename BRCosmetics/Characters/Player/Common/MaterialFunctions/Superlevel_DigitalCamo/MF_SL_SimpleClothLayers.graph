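# Material function that builds a simple cloth layer as material attributes, filed under player-character Superlevel digital camo.
# A scalar mask input blends two color parameters into base color, with parameterized roughness and specular and zero subsurface color.
# A tiled cloth normal map, replaceable by a flat normal through a static switch, sets the normal.
# The transformed normal also drives a Fresnel that adds a faint emissive rim.
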MaterialFunction /BRCosmetics/Characters/Player/Common/MaterialFunctions/Superlevel_DigitalCamo/MF_SL_SimpleClothLayers
graph {nodes [
  n1 = Constant() {r=1}
  n2 = TextureCoordinate()
  n3 = ScalarParameter() {defaultValue=12, parameterName=ClothTiling}
  n4 = StaticSwitchParameter(
   True=Constant3Vector() {constant=<rgba 0 0 1 0>}
   False=TextureSample(UVs=Multiply(A=@n2, B=@n3)) {samplerType=SAMPLERTYPE_Normal, texture=&/BasicCloth_N}.RGB
  ) {parameterName=RemoveClothNormal}
  n5 = Transform(None=@n4)
  n6 = SetMaterialAttributes!(
   MaterialAttributes
   BaseColor=LinearInterpolate(
    A=VectorParameter() {parameterName=BlackColor, defaultValue=<rgba 0.03 0.03 0.03 1>}.RGB
    B=VectorParameter() {parameterName=WhiteColor, defaultValue=<rgba 0.4896 0.4896 0.4896 1>}.RGB
    Alpha=FunctionInput(Preview=@n1) {inputName=Black/White Mask, inputType=FunctionInput_Scalar}
   )
   Roughness=ScalarParameter() {defaultValue=0.9, parameterName=ClothRoughness}
   Normal=@n4
   Specular=ScalarParameter() {defaultValue=0.2, parameterName=ClothSpec}
   EmissiveColor=LinearInterpolate(B=ScalarParameter() {defaultValue=0.08, parameterName=EmissiveClothIntensity}, Alpha=Fresnel(Normal=@n5) {baseReflectFraction=0, exponent=7}) {constB=0.08}
   SubsurfaceColor=Constant()
  ) {attributeSetTypes [
    #1
    #2
    #3
    #4
    #5
    #6
   ]}
  FunctionOutput(None=@n6)
 ]}
