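# Material function for the Superlevel DigitalCamo player cosmetics that outputs material attributes for a simple fully metallic surface with two tintable metals and grime.
# The MetalLerp input blends color and roughness parameters from Metal B to Metal A.
# The Grime input, previewed and defaulted from a metal noise mask, then blends in a dirt color at reduced strength and a grime roughness at full strength.
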MaterialFunction /BRCosmetics/Characters/Player/Common/MaterialFunctions/Superlevel_DigitalCamo/MF_SL_SimpleMetals
graph {
 properties {libraryCategoriesText [Misc, ""]}
 nodes [
  n1 = VectorParameter() {parameterName=MetalB_Color, defaultValue=<rgba 0.9375 0.8463 0.7888 1>}
  n2 = VectorParameter() {parameterName=MetalA_Color, defaultValue=<rgba 0.9375 0.5844 0.1086 1>}
  n3 = FunctionInput(Preview=Constant()) {bUsePreviewValueAsDefault=true, inputName=MetalLerp, inputType=FunctionInput_Scalar}
  n4 = FunctionInput(Preview=TextureSample() {samplerType=SAMPLERTYPE_Masks, texture=/Game/Global/Textures/Default/Noise/T_Metal_Noise}.R) {bUsePreviewValueAsDefault=true, inputName=Grime, inputType=FunctionInput_Scalar}
  n5 = ScalarParameter() {defaultValue=0.1, parameterName=MetalB_Roughness}
  n6 = ScalarParameter() {defaultValue=0.25, parameterName=MetalA_Roughness}
  n7 = SetMaterialAttributes!(
   MaterialAttributes
   BaseColor=LinearInterpolate(
    A=LinearInterpolate(A=@n1.RGB, B=@n2.RGB, Alpha=@n3)
    B=VectorParameter() {parameterName=CirtColor, defaultValue=<rgba 0.0208 0.0168 0.0129 1>}.RGB
    Alpha=LinearInterpolate(Alpha=@n4) {constB=0.4}
   )
   Metallic=Constant() {r=1}
   Roughness=LinearInterpolate(A=LinearInterpolate(A=@n5, B=@n6, Alpha=@n3), B=ScalarParameter() {defaultValue=0.35, parameterName=MetalGrimeRoughness}, Alpha=@n4)
   Specular
   SubsurfaceColor=Constant()
  ) {attributeSetTypes [
    #1
    #2
    #3
    #4
    #5
   ]}
  FunctionOutput(None=@n7)
 ]
}
