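# Cloth-shaded character material for the Tier 1 Super Level skin of the male Mystic body.
# Two mask textures layer cloth, digital camo, metal and rubber functions over the PreFX base, whose normal is flattened first and angle-blended back afterward.
# Static switches select toon shading with animated visor, a Binary helm visor, or neither.
# A power-adjusted mask blends in emissive camo to prevent jaggies before the shared PostFX function.
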Material /BRCosmetics/Characters/Player/Male/Medium/Bodies/M_MED_Mystic/Materials/Superlevels/T1/M_MED_MYSTIC_Body_SL_T1
graph {
 properties {
  bAutomaticallySetUsageInEditor=false
  bUseMaterialAttributes=true
  bUsedWithClothing=true
  bUsedWithInstancedStaticMeshes=true
  bUsedWithMeshDeformer=true
  bUsedWithMorphTargets=true
  bUsedWithSkeletalMesh=true
  pixelDepthOffsetMode=PDOM_Legacy
  shadingModel=MSM_Cloth
  subsurfaceProfile=&SS_HeroSkin_02
  usedShadingModels=Cloth
 }
 nodes [
  n1 = MaterialFunctionCall() {materialFunction=&BaseFN_CharacterFunction_PreFX}
  n2 = TextureSampleParameter2D() {
   parameterName=Mask02
   samplerType=SAMPLERTYPE_Masks
   texture=/BRCosmetics/Characters/Player/Male/Medium/Bodies/M_MED_Mystic/Textures/Skins/Super_Levels/T_Mystic_-_Super_Levels_Body_SL_FX_2
  }
  n3 = TextureSampleParameter2D() {
   parameterName=Mask01
   samplerType=SAMPLERTYPE_Masks
   texture=/BRCosmetics/Characters/Player/Male/Medium/Bodies/M_MED_Mystic/Textures/Skins/Super_Levels/T_Mystic_-_Super_Levels_Body_SL_FX
  }
  n4 = Max(A=@n2.B, B=Max(A=Max(A=@n3.R, B=@n3.R), B=@n3.G))
  n5 = OneMinus(None=Max(A=@n2.G, B=Max(A=@n2.R, B=@n4)))
  n6 = GetMaterialAttributes(None=@n1.Result) {attributeGetTypes [#1]}
  n7 = SetMaterialAttributes!(MaterialAttributes=@n6.MaterialAttributes, Normal=Constant3Vector() {constant=<rgba 0 0 1 0>}) {attributeSetTypes [#1]}
  n8 = BlendMaterialAttributes(
   A=@n7
   B=MaterialFunctionCall!(Black/White Mask=@n2.R) {materialFunction=/BRCosmetics/Characters/Player/Common/MaterialFunctions/Superlevel_DigitalCamo/MF_SL_SimpleClothLayers}.Result
   Alpha=@n3.R
  )
  n9 = MaterialFunctionCall!(PreSkinnedLocal=@n1.Pre skinned Local Position, Fresnel=@n1.Pre skinned Local Position, GMap=@n1.Pre skinned Local Position, In=@n7) {materialFunction=/BRCosmetics/Characters/Player/Common/MaterialFunctions/Superlevel_DigitalCamo/MF_SL_DigitalCamo}
  n10 = BlendMaterialAttributes(
   A=BlendMaterialAttributes(
    A=BlendMaterialAttributes(A=@n8, B=@n9.Result, Alpha=@n3.G)
    B=MaterialFunctionCall!(Grime, MetalLerp=@n2.R) {materialFunction=/BRCosmetics/Characters/Player/Common/MaterialFunctions/Superlevel_DigitalCamo/MF_SL_SimpleMetals}.Result
    Alpha=@n3.B
   )
   B=MaterialFunctionCall() {materialFunction=/BRCosmetics/Characters/Player/Common/MaterialFunctions/Superlevel_DigitalCamo/MF_SL_BasicRubber}.Result
   Alpha=@n2.B
  )
  n11 = StaticSwitchParameter(
   True=MaterialFunctionCall!(In=MaterialFunctionCall!(SkinMask=@n5, In=@n10) {materialFunction=/Game/Packages/Fortress_SharedMaterials/Anime_Master_Material/MF_Toonify}.Result) {materialFunction=/Game/Packages/Fortress_SharedMaterials/Anime_Master_Material/MF_AnimatedVisor}.Result
   False=StaticSwitchParameter(
    True=MaterialFunctionCall!(Mask=@n2.B, Material=@n10) {materialFunction=/Game/Characters/Player/Female/Medium/Bodies/F_MED_Binary/Materials/MF_Visor}.Result
    False=@n10
   ) {parameterName=isBinaryHelm}
  ) {parameterName=useAnimeShading}
  n12 = GetMaterialAttributes(None=BlendMaterialAttributes(
    A=@n11
    B=MaterialFunctionCall!(In=@n11, Fresnel=@n1.Fresnel) {materialFunction=/BRCosmetics/Characters/Player/Common/MaterialFunctions/Superlevel_DigitalCamo/MF_SL_DigitalCamo_Emissive}.Result
    Alpha=Power(Base=@n2.G) {desc=Prevents Jaggies}
   )) {attributeGetTypes [#1, #2, #3]}
  MaterialFunctionCall(
   AO=@n1.AO
   Distance Blend=@n1.AO
   Fresnel=@n1.AO
   Gmap=@n1.AO
   Pre Skinned Local Position=@n1.AO
   Material=SetMaterialAttributes!(
    MaterialAttributes=@n12.MaterialAttributes
    Normal=MaterialFunctionCall(BaseNormal=@n12.Normal, AdditionalNormal=@n6.Normal) {materialFunction=&BlendAngleCorrectedNormals}.Result
    Base Color=@n12.MaterialAttributes
   ) {attributeSetTypes [#1, #2]}
  ) {materialFunction=&BaseFN_CharacterFunction_PostFX}
 ]
}
